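# Sound cue for the completion of a small holdable button in the front-end UI, using the quiet variant of the sound.
# A single wave player feeds the quiet press-hold-complete wave through a Modulator at default settings, routed to the Fort_UI sound class at reduced volume.
# A concurrency override caps simultaneous instances, blocks new plays once the cap is reached, and sets a retrigger time to limit rapid repeats.
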SoundCue /Game/Sounds/Fort_UI_Sounds/FrontEnd/DefaultSounds/HoldableButtons/Holable_Quiet/Fortnite_UI_Default_Holdable_Small_Complete_Cue
root=Modulator_0
nodes [
 {id=Modulator_0, children [WavePlayer_0]}
 {
  id=WavePlayer_0
  properties {soundWaveAssetPtr=/Game/Sounds/Fort_UI_Sounds/FrontEnd/DefaultSounds/HoldableButtons/wav/UI_Button_Press_Hold_Quiet_Complete}
 }
]
properties {
 bOverrideConcurrency=true
 maxDistance=2097152
 soundClassObject=&Fort_UI
 volumeMultiplier=0.629
 concurrencyOverrides {maxCount=2, resolutionRule=PreventNew, retriggerTime=0.5}
}
