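# Sound cue for the release sound of a small, quiet holdable button in the default front-end UI.
# A Modulator feeds from a single Wave Player that plays the quiet press-and-hold release wave.
# The cue plays pitched up and attenuated in volume on the Fort_UI sound class.
# It overrides concurrency with a short retrigger time to limit rapid repeats.
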SoundCue /Game/Sounds/Fort_UI_Sounds/FrontEnd/DefaultSounds/HoldableButtons/Holable_Quiet/Fortnite_UI_Default_Holdable_Small_Release_Cue
root=Modulator_1
nodes [
 {id=Modulator_1, children [WavePlayer_1]}
 {
  id=WavePlayer_1
  properties {soundWaveAssetPtr=/Game/Sounds/Fort_UI_Sounds/FrontEnd/DefaultSounds/HoldableButtons/wav/UI_Button_Press_Hold_Quiet_Release}
 }
]
properties {
 bOverrideConcurrency=true
 maxDistance=2097152
 pitchMultiplier=1.5
 soundClassObject=&Fort_UI
 volumeMultiplier=0.378
 concurrencyOverrides {retriggerTime=0.1}
}
